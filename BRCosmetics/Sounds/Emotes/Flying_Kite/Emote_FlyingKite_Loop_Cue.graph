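# Looping SoundCue for the Flying Kite emote, mixing a steady wind loop with kite movement sounds and a one-shot poof-in.
# The movement branch loops a random pick of five kite waves through a modulator and random delay, mixed quieter than the wind.
# It uses emote attenuation, sound class and foley concurrency, keeps playing when silent, and has a send to the music audio-analysis submix.
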SoundCue /BRCosmetics/Sounds/Emotes/Flying_Kite/Emote_FlyingKite_Loop_Cue
root=Mixer_0
nodes [
 {id=Delay_0, children [Modulator_0], properties {delayMax=1.5}}
 {id=Looping_0, children [Delay_0]}
 {id=Mixer_0, children [WavePlayer_0,Looping_0,WavePlayer_6], properties {inputVolume [1,0.3,0.7]}}
 {id=Modulator_0, children [Random_0]}
 {id=Random_0, children [WavePlayer_1,WavePlayer_2,WavePlayer_3,WavePlayer_4,WavePlayer_5], properties {weights [1,1,1,1,1]}}
 {id=WavePlayer_0, properties {bLooping=true, soundWaveAssetPtr=&/Emote_FlyingKite_WindLoop}}
 {id=WavePlayer_1, properties {soundWaveAssetPtr=&/Emote_FlyingKite_KiteMovement_01}}
 {id=WavePlayer_2, properties {soundWaveAssetPtr=&/Emote_FlyingKite_KiteMovement_02}}
 {id=WavePlayer_3, properties {soundWaveAssetPtr=&/Emote_FlyingKite_KiteMovement_03}}
 {id=WavePlayer_4, properties {soundWaveAssetPtr=&/Emote_FlyingKite_KiteMovement_04}}
 {id=WavePlayer_5, properties {soundWaveAssetPtr=&/Emote_FlyingKite_KiteMovement_05}}
 {id=WavePlayer_6, properties {soundWaveAssetPtr=&/Emote_FlyingKite_PoofIn}}
]
properties {
 attenuationSettings=&Fort_Emotes_Athena_Att
 maxDistance=1600
 soundClassObject=&Fort_Music_Emote
 virtualizationMode=PlayWhenSilent
 volumeMultiplier=0.5
 concurrencySet [/Game/Sounds/Concurrency_Rules/Athena_EmoteFoley_Conc]
 soundSubmixSends [{minSendLevel=0, soundSubmix=&Fortnite_AudioAnalysis_Music}]
}
